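# Sound cue for the Clyde boss ClayPlug that plays a music stinger for friendly interaction.
# A licensed content switcher feeds the output, with the stinger wave player on its third input and the first two inputs empty.
# It routes to the NPC music sound class, borrows the shared NPC alerted-voice attenuation, and boosts volume slightly.
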SoundCue /Clyde/Audio/Friendly_Interact/Clyde_Boss_ClayPlug_FriendlyInteract
root=FortSoundNodeLicensedContentSwitcher_0
nodes [
 {id=FortSoundNodeLicensedContentSwitcher_0, class=FortSoundNodeLicensedContentSwitcher, children [null,null,WavePlayer_0]}
 {id=WavePlayer_0, properties {soundWaveAssetPtr=&/MX_IP_ClayPlug_FriendlyInteract_Stinger}}
]
properties {
 attenuationSettings=/Game/Sounds/Fort_Human_VO_BR/NPCs/Shared/NPC_VO_Alerted_Att
 maxDistance=12500
 soundClassObject=&Fort_Music_NPC
 volumeMultiplier=1.2
}
